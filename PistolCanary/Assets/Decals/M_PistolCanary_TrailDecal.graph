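# Translucent, lit deferred decal material for trail marks, built on a trail mask texture and tinted by a BaseColor parameter or, via static switch, the decal component's color.
# A world-aligned noise texture is shaped by contrast, fill and bias into a saturated mask, apparently subtracted from the trail mask to break it up.
# The same "Use World Aligned Texture?" switch selects a flattened world-aligned detail normal over a flat one.
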Material /PistolCanary/Assets/Decals/M_PistolCanary_TrailDecal
graph {
 expressions {
  MaterialExpressionAdd_1 {
   class=MaterialExpressionAdd
   inputs [
    A {from=MaterialExpressionMaterialFunctionCall_1, output=Result}
    B {from=MaterialExpressionScalarParameter_5}
   ]
  }
  MaterialExpressionConstant3Vector_1 {class=MaterialExpressionConstant3Vector, properties {constant=<rgba 0 0 1 0>}}
  MaterialExpressionConstant_0 {class=MaterialExpressionConstant}
  MaterialExpressionConvert_0 {
   class=MaterialExpressionConvert
   inputs [
    None {from=MaterialExpressionScalarParameter_6}
    None {from=MaterialExpressionScalarParameter_6}
    None {from=MaterialExpressionScalarParameter_6}
   ]
   properties {
    nodeName=Make Float3
    convertInputs [
     {
      type=Scalar
      expressionInput {
       expression=&self:MaterialExpressionScalarParameter_6
       inputName=None
       mask=0
       maskA=0
       maskB=0
       maskG=0
       maskR=0
       outputIndex=0
      }
     }
     {
      type=Scalar
      expressionInput {
       expression=&self:MaterialExpressionScalarParameter_6
       inputName=None
       mask=0
       maskA=0
       maskB=0
       maskG=0
       maskR=0
       outputIndex=0
      }
     }
     {
      type=Scalar
      expressionInput {
       expression=&self:MaterialExpressionScalarParameter_6
       inputName=None
       mask=0
       maskA=0
       maskB=0
       maskG=0
       maskR=0
       outputIndex=0
      }
     }
    ]
    convertOutputs [{type=Vector3}]
   }
  }
  MaterialExpressionConvert_1 {
   class=MaterialExpressionConvert
   inputs [
    None {from=MaterialExpressionScalarParameter_8}
    None {from=MaterialExpressionScalarParameter_8}
    None {from=MaterialExpressionScalarParameter_8}
   ]
   properties {
    nodeName=Make Float3
    convertInputs [
     {
      type=Scalar
      expressionInput {
       expression=&self:MaterialExpressionScalarParameter_8
       inputName=None
       mask=0
       maskA=0
       maskB=0
       maskG=0
       maskR=0
       outputIndex=0
      }
     }
     {
      type=Scalar
      expressionInput {
       expression=&self:MaterialExpressionScalarParameter_8
       inputName=None
       mask=0
       maskA=0
       maskB=0
       maskG=0
       maskR=0
       outputIndex=0
      }
     }
     {
      type=Scalar
      expressionInput {
       expression=&self:MaterialExpressionScalarParameter_8
       inputName=None
       mask=0
       maskA=0
       maskB=0
       maskG=0
       maskR=0
       outputIndex=0
      }
     }
    ]
    convertOutputs [{type=Vector3}]
   }
  }
  MaterialExpressionDecalColor_0 {class=MaterialExpressionDecalColor, inputs [], properties {}}
  MaterialExpressionMaterialFunctionCall_0 {
   class=MaterialExpressionMaterialFunctionCall
   inputs [
    TextureObject {from=MaterialExpressionTextureObject_2}
    TextureSize {from=MaterialExpressionConvert_0}
    WorldPosition {from=""}
    Export Float 4 {from=""}
    World Space Normal {from=""}
    ProjectionTransitionContrast {from=""}
   ]
   properties {materialFunction=&WorldAlignedTexture}
  }
  MaterialExpressionMaterialFunctionCall_1 {
   class=MaterialExpressionMaterialFunctionCall
   inputs [
    In {from=MaterialExpressionMaterialFunctionCall_0, output=XYZ Texture}
    Contrast {from=MaterialExpressionScalarParameter_2}
   ]
   properties {materialFunction=&CheapContrast}
  }
  MaterialExpressionMaterialFunctionCall_2 {
   class=MaterialExpressionMaterialFunctionCall
   inputs [
    TextureObject {from=MaterialExpressionTextureObject_3}
    TextureSize {from=MaterialExpressionConvert_1}
    WorldSpace {from=""}
    Normal {from=""}
    ProjectionTransitionContrast {from=""}
    WorldPosition {from=""}
    Use High Quality Normals {from=""}
   ]
   properties {materialFunction=&WorldAlignedNormal}
  }
  MaterialExpressionMaterialFunctionCall_3 {
   class=MaterialExpressionMaterialFunctionCall
   inputs [
    Normal {from=MaterialExpressionMaterialFunctionCall_2, output=XYZ Texture}
    Flatness {from=MaterialExpressionScalarParameter_9}
   ]
   properties {materialFunction=&FlattenNormal}
  }
  MaterialExpressionMultiply_0 {
   class=MaterialExpressionMultiply
   inputs [
    A {from=MaterialExpressionNamedRerouteUsage_1}
    B {from=MaterialExpressionStaticSwitchParameter_0}
   ]
  }
  MaterialExpressionMultiply_2 {
   class=MaterialExpressionMultiply
   inputs [
    A {from=MaterialExpressionAdd_1}
    B {from=MaterialExpressionScalarParameter_7}
   ]
  }
  MaterialExpressionNamedRerouteDeclaration_0 {
   class=MaterialExpressionNamedRerouteDeclaration
   inputs [None {from=MaterialExpressionTextureSampleParameter2D_0, output=RGB}]
   properties {name=MainTexture, nodeColor=<rgba 1 0 0.1172 1>}
  }
  MaterialExpressionNamedRerouteDeclaration_2 {
   class=MaterialExpressionNamedRerouteDeclaration
   inputs [None {from=MaterialExpressionSaturate_0}]
   properties {name=WAT - Mask, nodeColor=<rgba 0 0.2266 1 1>}
  }
  MaterialExpressionNamedRerouteDeclaration_3 {
   class=MaterialExpressionNamedRerouteDeclaration
   inputs [None {from=MaterialExpressionStaticSwitchParameter_2}]
   properties {name=Normal, nodeColor=<rgba 0.1719 1 0 1>}
  }
  MaterialExpressionNamedRerouteUsage_0 {class=MaterialExpressionNamedRerouteUsage}
  MaterialExpressionNamedRerouteUsage_1 {class=MaterialExpressionNamedRerouteUsage}
  MaterialExpressionNamedRerouteUsage_3 {class=MaterialExpressionNamedRerouteUsage}
  MaterialExpressionNamedRerouteUsage_4 {class=MaterialExpressionNamedRerouteUsage}
  MaterialExpressionNamedRerouteUsage_6 {class=MaterialExpressionNamedRerouteUsage}
  MaterialExpressionSaturate_0 {class=MaterialExpressionSaturate, inputs [None {from=MaterialExpressionMultiply_2}]}
  MaterialExpressionScalarParameter_0 {class=MaterialExpressionScalarParameter, properties {defaultValue=0.5, parameterName=Specular}}
  MaterialExpressionScalarParameter_1 {class=MaterialExpressionScalarParameter, properties {defaultValue=0.5, parameterName=Roughness}}
  MaterialExpressionScalarParameter_2 {class=MaterialExpressionScalarParameter, properties {parameterName=WAT - Contrast}}
  MaterialExpressionScalarParameter_5 {class=MaterialExpressionScalarParameter, properties {defaultValue=-0.25, parameterName=WAT - Fill}}
  MaterialExpressionScalarParameter_6 {class=MaterialExpressionScalarParameter, properties {defaultValue=500, parameterName=WAT-TextureSize}}
  MaterialExpressionScalarParameter_7 {class=MaterialExpressionScalarParameter, properties {defaultValue=1, parameterName=WAT- Bias}}
  MaterialExpressionScalarParameter_8 {class=MaterialExpressionScalarParameter, properties {defaultValue=250, parameterName=WAN-TextureSize}}
  MaterialExpressionScalarParameter_9 {class=MaterialExpressionScalarParameter, properties {defaultValue=0.7, parameterName=WAN - NormalFlatten}}
  MaterialExpressionStaticSwitchParameter_0 {
   class=MaterialExpressionStaticSwitchParameter
   inputs [
    True {from=MaterialExpressionDecalColor_0, output=RGB}
    False {from=MaterialExpressionVectorParameter_0, output=RGB}
   ]
   properties {parameterName=UseCustomColor?}
  }
  MaterialExpressionStaticSwitchParameter_1 {
   class=MaterialExpressionStaticSwitchParameter
   inputs [
    True {from=MaterialExpressionSubtract_0, output=""}
    False {from=MaterialExpressionNamedRerouteUsage_3, output=""}
   ]
   properties {parameterName=Use World Aligned Texture?}
  }
  MaterialExpressionStaticSwitchParameter_2 {
   class=MaterialExpressionStaticSwitchParameter
   inputs [
    True {from=MaterialExpressionMaterialFunctionCall_3, output=Result}
    False {from=MaterialExpressionConstant3Vector_1, output=""}
   ]
   properties {parameterName=Use World Aligned Texture?}
  }
  MaterialExpressionSubtract_0 {
   class=MaterialExpressionSubtract
   inputs [
    A {from=MaterialExpressionNamedRerouteUsage_0}
    B {from=MaterialExpressionNamedRerouteUsage_4}
   ]
  }
  MaterialExpressionTextureObject_2 {class=MaterialExpressionTextureObject, inputs [], properties {texture=/PistolCanary/Assets/Decals/Textures/PistolCanary_NoiseClouds}}
  MaterialExpressionTextureObject_3 {
   class=MaterialExpressionTextureObject
   inputs []
   properties {samplerType=SAMPLERTYPE_Normal, texture=/PistolCanary/Assets/Textures/Noisepack_Detail_N}
  }
  MaterialExpressionTextureSampleParameter2D_0 {
   class=MaterialExpressionTextureSampleParameter2D
   inputs [
    UVs {from="", output=""}
    Apply View MipBias {from="", output=""}
   ]
   properties {parameterName=Texture, texture=/PistolCanary/Assets/Decals/Textures/T_Trail_Mask_01}
  }
  MaterialExpressionVectorParameter_0 {class=MaterialExpressionVectorParameter, properties {parameterName=BaseColor, defaultValue=<rgba 0.3864 0.1746 0.1248 1>}}
 }
 properties {bAutomaticallySetUsageInEditor=false, blendMode=BLEND_Translucent, materialDomain=MD_DeferredDecal, usedShadingModels=Default Lit}
}
